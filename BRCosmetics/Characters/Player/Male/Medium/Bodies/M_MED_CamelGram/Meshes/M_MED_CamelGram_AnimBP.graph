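# Animation Blueprint for the CamelGram medium male player body cosmetic, following the attached parent mesh's pose and adjusting its accessory bones.
# After Copy Pose From Mesh, chained weighted multi-target Constraint nodes position the holster from right thigh, pelvis, groin and glute bones.
# Ammo and belt accessory bones are constrained mainly to pelvis and spine, most with slight left-leg influence.
# The CamelGram Control Rig then finishes the pose; the event graph holds no logic.
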AnimBlueprint /BRCosmetics/Characters/Player/Male/Medium/Bodies/M_MED_CamelGram/Meshes/M_MED_CamelGram_AnimBP
target_skeleton=&/M_MED_CamelGram_Skeleton
graphs [
 AnimGraph {nodes [
   n1 = "Constraint - holster (PPPPP)"!(
    ComponentPose=Local To Component(LocalPose=Copy Pose From Mesh() {bUseAttachedParent=true})
    Alpha=1
    ConstraintWeights_0=0.785
    ConstraintWeights_1:float=0.1162
    ConstraintWeights_2:float=0.0555
    ConstraintWeights_3:float=0.0415
    ConstraintWeights_4:float=0.0018
   ) -> (Pose) {
    alphaBoolBlend {blendOption=Linear}
    boneToModify {boneName=holster}
    constraintSetup [
     {targetBone {boneName=thigh_twist_01_r}}
     {targetBone {boneName=thigh_r}}
     {targetBone {boneName=pelvis}}
     {targetBone {boneName=deform_groin_r}}
     {targetBone {boneName=deform_glute_r}}
    ]
    constraintWeights [1,1,1,1,1]
   }
   n2 = "Constraint - base_ammo_01 (PPPPPP)"!(
    ComponentPose=@n1
    Alpha=1
    ConstraintWeights_0=0.7447
    ConstraintWeights_1:float=0.1004
    ConstraintWeights_2:float=0.0864
    ConstraintWeights_3:float=0.0588
    ConstraintWeights_4:float=0.0056
    ConstraintWeights_5:float=0.0042
   ) -> (Pose) {
    alphaBoolBlend {blendOption=Linear}
    boneToModify {boneName=base_ammo_01}
    constraintSetup [
     {targetBone {boneName=pelvis}}
     {targetBone {boneName=spine_02}}
     {targetBone {boneName=spine_01}}
     {targetBone {boneName=spine_03}}
     {targetBone {boneName=deform_groin_l}}
     {targetBone {boneName=thigh_twist_01_l}}
    ]
    constraintWeights [1,1,1,1,1,1]
   }
   n3 = "Constraint - base_ammo_02 (PPPPPP)"!(
    ComponentPose=@n2
    Alpha=1
    ConstraintWeights_0=0.757
    ConstraintWeights_1:float=0.0932
    ConstraintWeights_2:float=0.0805
    ConstraintWeights_3:float=0.0538
    ConstraintWeights_4:float=0.0087
    ConstraintWeights_5:float=0.0069
   ) -> (Pose) {
    alphaBoolBlend {blendOption=Linear}
    boneToModify {boneName=base_ammo_02}
    constraintSetup [
     {targetBone {boneName=pelvis}}
     {targetBone {boneName=spine_02}}
     {targetBone {boneName=spine_01}}
     {targetBone {boneName=spine_03}}
     {targetBone {boneName=deform_groin_l}}
     {targetBone {boneName=thigh_twist_01_l}}
    ]
    constraintWeights [1,1,1,1,1,1]
   }
   n4 = "Constraint - base_belt_acc_01 (PPPPPP)"!(
    ComponentPose=@n3
    Alpha=1
    ConstraintWeights_0=0.7824
    ConstraintWeights_1:float=0.0846
    ConstraintWeights_2:float=0.0726
    ConstraintWeights_3:float=0.0472
    ConstraintWeights_4:float=0.009
    ConstraintWeights_5:float=0.0041
   ) -> (Pose) {
    alphaBoolBlend {blendOption=Linear}
    boneToModify {boneName=base_belt_acc_01}
    constraintSetup [
     {targetBone {boneName=pelvis}}
     {targetBone {boneName=spine_02}}
     {targetBone {boneName=spine_01}}
     {targetBone {boneName=spine_03}}
     {targetBone {boneName=thigh_twist_01_l}}
     {targetBone {boneName=deform_groin_l}}
    ]
    constraintWeights [1,1,1,1,1,1]
   }
   n5 = "Constraint - base_belt_acc_02 (PPPPP)"!(
    ComponentPose=@n4
    Alpha=1
    ConstraintWeights_0=0.7774
    ConstraintWeights_1:float=0.0904
    ConstraintWeights_2:float=0.0771
    ConstraintWeights_3:float=0.0504
    ConstraintWeights_4:float=0.0048
   ) -> (Pose) {
    alphaBoolBlend {blendOption=Linear}
    boneToModify {boneName=base_belt_acc_02}
    constraintSetup [
     {targetBone {boneName=pelvis}}
     {targetBone {boneName=spine_02}}
     {targetBone {boneName=spine_01}}
     {targetBone {boneName=spine_03}}
     {targetBone {boneName=thigh_twist_01_l}}
    ]
    constraintWeights [1,1,1,1,1]
   }
   n6 = "Constraint - base_belt_acc_03 (PPPP)"!(
    ComponentPose=@n5
    Alpha=1
    ConstraintWeights_0=0.782
    ConstraintWeights_1:float=0.0898
    ConstraintWeights_2:float=0.077
    ConstraintWeights_3:float=0.0511
   ) -> (Pose) {
    alphaBoolBlend {blendOption=Linear}
    boneToModify {boneName=base_belt_acc_03}
    constraintSetup [{targetBone {boneName=pelvis}}, {targetBone {boneName=spine_02}}, {targetBone {boneName=spine_01}}, {targetBone {boneName=spine_03}}]
    constraintWeights [1,1,1,1]
   }
   n7 = Control Rig(Source=Component To Local(ComponentPose=@n6)) {
    alphaBoolBlend {blendOption=Linear}
    controlRigAssetReference {blueprintRigClass=&/M_MED_CamelGram_CtrlRig.M_MED_CamelGram_CtrlRig_C}
   }
   Output Pose(Result=@n7) {layerGroup=None}
  ]}
 EventGraph {nodes [Event BlueprintUpdateAnimation()]}
]
variables []
